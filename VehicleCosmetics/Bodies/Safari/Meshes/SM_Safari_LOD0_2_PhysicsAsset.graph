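# Physics asset for the Safari vehicle cosmetic body mesh, giving it a simplified collision shape.
# A single body on the frame bone holds one large capsule, rotated onto its side and offset from the bone origin.
# It has no other bodies or constraints, so it defines no articulated parts.
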PhysicsAsset /VehicleCosmetics/Bodies/Safari/Meshes/SM_Safari_LOD0_2_PhysicsAsset
bodies {frame [{shapeType=Capsule, length=197.782, radius=90.6582, center=<2.8054 -54.5858 -0.1333>, rotation=<rot 0 -87.2335 -89.9318>}]}
constraints []
